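# Level Sequence that animates the ValueEnjoy_M custom skin (V02) for the Battle Royale menu on a spawned FortPlayerMannequin.
# Control Rig tracks drive MR_Biped_v02 for the body, with no channels shown, and Fortnite_3L_Face_CtrlRig for keyed blinks, gaze, brows and mouth.
# The skeletal animation track is empty.
# At 30 fps, face keys repeat their opening values at the end, forming a loop of roughly 12 seconds.
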LevelSequence /BRCosmetics/Animation/Game/MainPlayer/Menu/BR/CustomSkins/_S41/ValueEnjoy_M/ValueEnjoy_Anim/V02/ValueEnjoy_Anim_v02
bindings [FortPlayerMannequin {
  id=#1
  possessed_class=FortPlayerMannequin
  tracks [
   {type=MovieSceneSpawnTrack, sections [{type=MovieSceneSpawnSection, channels [None {type=MovieSceneScriptingBoolChannel, default=true}]}]}
   {
    type=MovieScene3DTransformTrack
    sections [{
      type=MovieScene3DTransformSection
      channels [
       Location.X {type=MovieSceneScriptingDoubleChannel, default=0}
       Location.Y {type=MovieSceneScriptingDoubleChannel, default=0}
       Location.Z {type=MovieSceneScriptingDoubleChannel, default=0}
       Rotation.X {type=MovieSceneScriptingDoubleChannel, default=0}
       Rotation.Y {type=MovieSceneScriptingDoubleChannel, default=0}
       Rotation.Z {type=MovieSceneScriptingDoubleChannel, default=0}
       Scale.X {type=MovieSceneScriptingDoubleChannel, default=1}
       Scale.Y {type=MovieSceneScriptingDoubleChannel, default=1}
       Scale.Z {type=MovieSceneScriptingDoubleChannel, default=1}
      ]
     }]
   }
   {type=MovieSceneSkeletalAnimationTrack, sections []}
   {type=MovieSceneControlRigParameterTrack, display_name=MR_Biped_v02, sections [{type=MovieSceneControlRigParameterSection}]}
   {
    type=MovieSceneControlRigParameterTrack
    display_name=Fortnite_3L_Face_CtrlRig
    sections [{
      type=MovieSceneControlRigParameterSection
      channels [
       CTRL_C_eyesAim.Location.X {type=MovieSceneScriptingFloatChannel, default=0, const [0,2]}
       CTRL_C_eyesAim.Location.Y {type=MovieSceneScriptingFloatChannel, default=0, const [0,2]}
       CTRL_C_eyesAim.Location.Z {type=MovieSceneScriptingFloatChannel, default=0, const [0,2]}
       CTRL_C_eyesAim.Rotation.X {type=MovieSceneScriptingFloatChannel, default=0, const [0,2]}
       CTRL_C_eyesAim.Rotation.Y {type=MovieSceneScriptingFloatChannel, default=0, const [0,2]}
       CTRL_C_eyesAim.Rotation.Z {type=MovieSceneScriptingFloatChannel, default=0, const [0,2]}
       CTRL_C_eyesAim.Scale.X {type=MovieSceneScriptingFloatChannel, default=1, const [1,2]}
       CTRL_C_eyesAim.Scale.Y {type=MovieSceneScriptingFloatChannel, default=1, const [1,2]}
       CTRL_C_eyesAim.Scale.Z {type=MovieSceneScriptingFloatChannel, default=1, const [1,2]}
       CTRL_L_eyeAim.Location.X {type=MovieSceneScriptingFloatChannel, default=0, const [0,2]}
       CTRL_L_eyeAim.Location.Y {type=MovieSceneScriptingFloatChannel, default=0, const [0,2]}
       CTRL_L_eyeAim.Location.Z {type=MovieSceneScriptingFloatChannel, default=0, const [0,2]}
       CTRL_L_eyeAim.Rotation.X {type=MovieSceneScriptingFloatChannel, default=0, const [0,2]}
       CTRL_L_eyeAim.Rotation.Y {type=MovieSceneScriptingFloatChannel, default=0, const [0,2]}
       CTRL_L_eyeAim.Rotation.Z {type=MovieSceneScriptingFloatChannel, default=0, const [0,2]}
       CTRL_L_eyeAim.Scale.X {type=MovieSceneScriptingFloatChannel, default=1, const [1,2]}
       CTRL_L_eyeAim.Scale.Y {type=MovieSceneScriptingFloatChannel, default=1, const [1,2]}
       CTRL_L_eyeAim.Scale.Z {type=MovieSceneScriptingFloatChannel, default=1, const [1,2]}
       CTRL_R_eyeAim.Location.X {type=MovieSceneScriptingFloatChannel, default=0, const [0,2]}
       CTRL_R_eyeAim.Location.Y {type=MovieSceneScriptingFloatChannel, default=0, const [0,2]}
       CTRL_R_eyeAim.Location.Z {type=MovieSceneScriptingFloatChannel, default=0, const [0,2]}
       CTRL_R_eyeAim.Rotation.X {type=MovieSceneScriptingFloatChannel, default=0, const [0,2]}
       CTRL_R_eyeAim.Rotation.Y {type=MovieSceneScriptingFloatChannel, default=0, const [0,2]}
       CTRL_R_eyeAim.Rotation.Z {type=MovieSceneScriptingFloatChannel, default=0, const [0,2]}
       CTRL_R_eyeAim.Scale.X {type=MovieSceneScriptingFloatChannel, default=1, const [1,2]}
       CTRL_R_eyeAim.Scale.Y {type=MovieSceneScriptingFloatChannel, default=1, const [1,2]}
       CTRL_R_eyeAim.Scale.Z {type=MovieSceneScriptingFloatChannel, default=1, const [1,2]}
       CTRL_convergenceSwitch {type=MovieSceneScriptingFloatChannel, default=3.4028234663852886e+38, const [0,6]}
       CTRL_Master.Location.X {type=MovieSceneScriptingFloatChannel, default=35.1064, const [35.1064,3]}
       CTRL_Master.Location.Y {type=MovieSceneScriptingFloatChannel, default=0, const [0,3]}
       CTRL_Master.Location.Z {type=MovieSceneScriptingFloatChannel, default=0, const [0,3]}
       CTRL_Master.Rotation.X {type=MovieSceneScriptingFloatChannel, default=0, const [0,3]}
       CTRL_Master.Rotation.Y {type=MovieSceneScriptingFloatChannel, default=0, const [0,3]}
       CTRL_Master.Rotation.Z {type=MovieSceneScriptingFloatChannel, default=0, const [0,3]}
       CTRL_Master.Scale.X {type=MovieSceneScriptingFloatChannel, default=1, const [1,3]}
       CTRL_Master.Scale.Y {type=MovieSceneScriptingFloatChannel, default=1, const [1,3]}
       CTRL_Master.Scale.Z {type=MovieSceneScriptingFloatChannel, default=1, const [1,3]}
       CTRL_C_jaw.X {
        type=MovieSceneScriptingFloatChannel
        default=0
        thinned_from=39
        keys [[-10,0],[165,0],[168,-0.0197],[174,-0.0799],[179,-0.1008],[182,-0.1077],[198,-0.1077],[202,-0.0945],[205,-0.0653],[207,-0.0277],[208,-0.0054],[213,0],[380,0]]
       }
       CTRL_C_jaw.Y {
        type=MovieSceneScriptingFloatChannel
        default=0
        thinned_from=39
        keys [[-10,0],[159,0],[165,0.0353],[168,0.0498],[174,0.0935],[179,0.0577],[182,0.0517],[198,0.0517],[202,0.0453],[205,0.0313],[207,0.0037],[208,0.0026],[213,0],[380,0]]
       }
       CTRL_C_jaw_openExtreme {
        type=MovieSceneScriptingFloatChannel
        default=3.4028234663852886e+38
        thinned_from=32
        keys [[-10,0],[168,0],[174,1],[198,1],[202,0.877],[205,0.6064],[207,0.401],[208,0.0501],[213,0],[407,0]]
       }
       CTRL_C_jaw_fwdBack {
        type=MovieSceneScriptingFloatChannel
        default=3.4028234663852886e+38
        thinned_from=39
        keys [[-10,0],[159,0],[165,0.2413],[168,0.2606],[174,0.9723],[179,1],[198,1],[202,0.877],[205,0.6064],[207,0.401],[208,0.0501],[213,0],[380,0]]
       }
       CTRL_R_jaw_ChinRaiseD {
        type=MovieSceneScriptingFloatChannel
        default=3.4028234663852886e+38
        thinned_from=39
        keys [[-10,0],[113,0],[123,0.0791],[137,0.1054],[150,0.1054],[159,0.1296],[165,0.0096],[168,0],[198,0],[202,0.0368],[205,0.1177],[207,0.1791],[208,0.284],[213,0.299],[231,0.299],[254,0.2984],[264,0.0668],[290,0],[380,0]]
       }
       CTRL_R_jaw_ChinRaiseU {
        type=MovieSceneScriptingFloatChannel
        default=3.4028234663852886e+38
        thinned_from=39
        keys [[-10,0],[113,0],[123,0.0791],[137,0.1054],[150,0.1054],[159,0.1296],[165,0.2763],[168,0.2874],[174,0.5194],[198,0.5194],[202,0.4923],[205,0.4327],[207,0.3874],[208,0.3101],[213,0.299],[231,0.299],[254,0.2984],[264,0.0668],[290,0],[380,0]]
       }
       CTRL_L_jaw_ChinRaiseU {
        type=MovieSceneScriptingFloatChannel
        default=3.4028234663852886e+38
        thinned_from=39
        keys [[-10,0],[113,0],[123,0.0791],[137,0.1054],[150,0.1054],[159,0.1296],[165,0.4911],[168,0.5194],[174,0.6223],[179,0.6423],[198,0.6423],[202,0.6001],[205,0.5072],[207,0.4367],[208,0.3162],[213,0.299],[231,0.299],[254,0.2984],[264,0.0668],[290,0],[380,0]]
       }
       CTRL_L_jaw_ChinRaiseD {
        type=MovieSceneScriptingFloatChannel
        default=3.4028234663852886e+38
        thinned_from=39
        keys [[-10,0],[113,0],[123,0.0791],[137,0.1054],[150,0.1054],[159,0.1296],[165,0.0096],[168,0],[198,0],[202,0.0368],[205,0.1177],[207,0.1791],[208,0.284],[213,0.299],[231,0.299],[254,0.2984],[264,0.0668],[290,0],[380,0]]
       }
       CTRL_R_mouth_lipsBlow {
        type=MovieSceneScriptingFloatChannel
        default=3.4028234663852886e+38
        thinned_from=39
        keys [[-10,0],[168,0],[174,0.3309],[198,0.3309],[202,0.2902],[205,0.2006],[207,0.1327],[208,0.0166],[213,0],[380,0]]
       }
       CTRL_L_mouth_lipsTogetherU {type=MovieSceneScriptingFloatChannel, default=3.4028234663852886e+38, const [0,39]}
       CTRL_L_mouth_lipsTogetherD {type=MovieSceneScriptingFloatChannel, default=3.4028234663852886e+38, const [0,39]}
       CTRL_R_mouth_lipsTogetherU {type=MovieSceneScriptingFloatChannel, default=3.4028234663852886e+38, const [0,39]}
       CTRL_L_mouth_lipsBlow {
        type=MovieSceneScriptingFloatChannel
        default=3.4028234663852886e+38
        thinned_from=39
        keys [[-10,0],[168,0],[174,0.3309],[198,0.3309],[202,0.2902],[205,0.2006],[207,0.1327],[208,0.0166],[213,0],[380,0]]
       }
       CTRL_R_mouth_lipsTogetherD {type=MovieSceneScriptingFloatChannel, default=3.4028234663852886e+38, const [0,39]}
       CTRL_C_tongue_roll.X {type=MovieSceneScriptingFloatChannel, default=0, const [0,39]}
       CTRL_C_tongue_roll.Y {type=MovieSceneScriptingFloatChannel, default=0, thinned_from=39, keys [[-10,0],[168,0],[174,-0.0212],[179,0],[380,0]]}
       CTRL_C_tongue.X {type=MovieSceneScriptingFloatChannel, default=0, const [0,39]}
       CTRL_C_tongue.Y {type=MovieSceneScriptingFloatChannel, default=0, thinned_from=39, keys [[-10,0],[168,0],[174,-0.0212],[179,0],[380,0]]}
       CTRL_C_tongue_narrowWide {type=MovieSceneScriptingFloatChannel, default=3.4028234663852886e+38, thinned_from=39, keys [[-10,0],[168,0],[174,0.1305],[179,0],[380,0]]}
       CTRL_C_tongue_inOut {
        type=MovieSceneScriptingFloatChannel
        default=3.4028234663852886e+38
        thinned_from=38
        keys [[-10,0],[150,0],[159,0.6394],[165,0.0474],[168,0],[380,0]]
       }
       CTRL_R_nose.X {type=MovieSceneScriptingFloatChannel, default=0, const [0,53]}
       CTRL_R_nose.Y {type=MovieSceneScriptingFloatChannel, default=0, const [0,53]}
       CTRL_L_nose.X {type=MovieSceneScriptingFloatChannel, default=0, const [0,53]}
       CTRL_L_nose.Y {type=MovieSceneScriptingFloatChannel, default=0, const [0,53]}
       CTRL_L_eye.X {type=MovieSceneScriptingFloatChannel, default=0, const [0,53]}
       CTRL_L_eye.Y {type=MovieSceneScriptingFloatChannel, default=-0.1201, thinned_from=53, keys [[-10,-0.1201],[380,-0.1201]]}
       CTRL_L_eye_blink {
        type=MovieSceneScriptingFloatChannel
        default=3.4028234663852886e+38
        thinned_from=61
        keys [[-10,1],[-5,0.3583],[0,0.1859],[7,0.0457],[19,0],[31,0],[34,0.2235],[36,1],[39,1],[50,0],[106,0],[110,1],[113,1],[116,0.6064],[120,0.0758],[123,0],[174,0],[175,-0.0025],[177,-0.1704],[183,-0.2216],[193,-0.229],[210,0],[215,0.1066],[219,1],[221,1],[224,0.6064],[228,0.0357],[232,0],[236,0.0727],[241,0.0982],[253,0.0982],[256,0.2366],[258,1],[263,0],[316,0],[320,0.9961],[327,0.1455],[340,-0.0345],[355,0],[358,0.2235],[360,1],[363,1],[368,0.3583],[373,0.1859],[380,0.0457]]
       }
       CTRL_R_eye.X {
        type=MovieSceneScriptingFloatChannel
        default=0
        thinned_from=55
        keys [[-10,0],[159,0],[165,-0.0879],[168,-0.0879],[172,0],[177,0],[183,-0.0243],[193,-0.0278],[210,-0.0251],[220,-0.0251],[221,0],[380,0]]
       }
       CTRL_R_eye.Y {
        type=MovieSceneScriptingFloatChannel
        default=0
        thinned_from=55
        keys [[-10,0],[159,0],[165,-0.2196],[168,-0.2196],[172,-0.406],[174,-0.3295],[175,-0.3295],[177,-0.2497],[183,-0.2497],[193,-0.2516],[220,-0.2516],[221,-0.0628],[320,-0.0628],[327,-0.0292],[340,0],[380,0]]
       }
       CTRL_R_eye_blink {
        type=MovieSceneScriptingFloatChannel
        default=3.4028234663852886e+38
        thinned_from=62
        keys [[-10,1],[-5,0.3583],[0,0.1859],[7,0.0457],[19,0],[30,0],[33,0.2235],[35,1],[38,1],[50,0],[107,0],[111,1],[114,1],[117,0.6064],[121,0.0758],[124,0],[174,0],[175,-0.0025],[177,-0.1704],[183,-0.2216],[193,-0.229],[210,0],[215,0.1066],[219,1],[222,1],[225,0.6064],[229,0.0357],[232,0],[236,0.0727],[241,0.0982],[253,0.0982],[256,0.2366],[259,1],[264,0],[315,0],[316,0.0557],[319,0.9961],[327,0.1455],[340,-0.0345],[355,0],[358,0.2235],[360,1],[363,1],[368,0.3583],[373,0.1859],[380,0.0457]]
       }
       CTRL_C_eye.X {
        type=MovieSceneScriptingFloatChannel
        default=0.1497
        thinned_from=71
        keys [[-10,0.1157],[0,0.1225],[24,0.1225],[25,0.1757],[26,0.1866],[31,0.1866],[32,0.1543],[34,-0.122],[35,-0.2913],[40,-0.3516],[59,-0.3516],[60,-0.3611],[61,-0.4527],[82,-0.4527],[83,-0.4363],[84,-0.3672],[85,-0.355],[90,-0.355],[91,-0.3404],[92,-0.24],[93,-0.0997],[95,-0.0462],[100,0.008],[102,0.008],[104,0.0834],[106,0.2993],[110,0.3762],[114,0.1312],[117,0.1302],[121,0.1235],[133,0.1235],[134,0.148],[135,0.1825],[136,0.194],[148,0.194],[149,0.1851],[150,0.1821],[159,0.3065],[165,0.3703],[168,0.3916],[172,0.415],[205,0.415],[216,0.5313],[224,0.2962],[249,0.2962],[251,0.396],[257,0.4293],[258,0.3371],[261,0.3136],[277,0.3136],[279,0.3269],[280,0.4207],[300,0.4207],[301,0.3599],[312,0.3599],[313,0.1859],[317,0.128],[324,-0.0417],[327,-0.109],[340,-0.072],[355,0.0007],[363,0.1157],[373,0.1225],[380,0.1225]]
       }
       CTRL_C_eye.Y {
        type=MovieSceneScriptingFloatChannel
        default=-0.0062
        thinned_from=71
        keys [[-10,-0.1838],[0,-0.0062],[24,-0.0062],[25,0.0243],[33,0.0243],[35,-0.0826],[39,-0.0215],[41,-0.0215],[50,-0.1488],[59,-0.1488],[60,-0.1335],[61,-0.0877],[82,-0.0877],[83,-0.0686],[84,-0.0255],[85,-0.0112],[104,-0.0112],[106,-0.0638],[109,-0.1838],[114,-0.1838],[117,-0.2632],[121,-0.0166],[124,0.1311],[133,0.1373],[134,0.1959],[135,0.265],[136,0.2881],[148,0.2881],[149,0.2274],[150,0.2072],[165,0.2072],[168,0.1475],[172,0.4557],[216,0.4557],[224,-0.0394],[300,-0.0394],[301,-0.0825],[320,-0.0825],[324,-0.3518],[327,-0.1838],[363,-0.1838],[373,-0.0062],[380,-0.0062]]
       }
       CTRL_C_eyesConvergence {type=MovieSceneScriptingFloatChannel, default=3.4028234663852886e+38, const [0,53]}
       CTRL_L_eye_cheekRaise {
        type=MovieSceneScriptingFloatChannel
        default=3.4028234663852886e+38
        thinned_from=62
        keys [[-10,0.6203],[111,0.6203],[114,0.4348],[117,0.0322],[121,-0.1217],[150,-0.1217],[156,0.4205],[159,1.4432],[165,0.6807],[168,0.7008],[172,0.6931],[174,0.6203],[175,0.5915],[177,0.6129],[183,0.6194],[219,0.6203],[222,0.0322],[225,-0.1116],[229,-0.1217],[232,-0.1217],[241,0],[253,0],[256,0.1535],[259,1],[264,0.2248],[320,0.2248],[327,0.2411],[340,0.6203],[380,0.6203]]
       }
       CTRL_L_eye_squintInner {type=MovieSceneScriptingFloatChannel, default=3.4028234663852886e+38, const [0,53]}
       CTRL_L_nose_wrinkleUpper {type=MovieSceneScriptingFloatChannel, default=3.4028234663852886e+38, const [0,53]}
       CTRL_R_nose_wrinkleUpper {type=MovieSceneScriptingFloatChannel, default=3.4028234663852886e+38, const [0,53]}
       CTRL_R_eye_squintInner {type=MovieSceneScriptingFloatChannel, default=3.4028234663852886e+38, const [0,53]}
       CTRL_R_eye_cheekRaise {
        type=MovieSceneScriptingFloatChannel
        default=3.4028234663852886e+38
        thinned_from=62
        keys [[-10,0.6203],[111,0.6203],[114,0.4348],[117,0.0322],[121,-0.1217],[150,-0.1217],[156,0.4205],[159,1.4432],[165,0.6203],[219,0.6203],[222,0.0322],[225,-0.1116],[229,-0.1217],[232,-0.1217],[241,0],[253,0],[256,0.1363],[259,0.8876],[264,0.1124],[320,0.1124],[327,0.213],[340,0.6203],[380,0.6203]]
       }
       CTRL_C_eyelidsReduce {
        type=MovieSceneScriptingFloatChannel
        default=3.4028234663852886e+38
        thinned_from=39
        keys [[-10,0],[253,0],[259,0.1201],[264,0.015],[274,0],[380,0]]
       }
       CTRL_R_brow_lateral {
        type=MovieSceneScriptingFloatChannel
        default=3.4028234663852886e+38
        thinned_from=40
        keys [[-10,0.5698],[107,0.5698],[111,0.6095],[117,0.7009],[123,0.6739],[137,0.649],[150,0.649],[156,0.6501],[159,0.6846],[162,0.9383],[165,0.6707],[168,0.5698],[178,0.5106],[183,0.3339],[193,0.3015],[210,0.4827],[231,0.6692],[243,0.7314],[322,0.7314],[329,0.6102],[340,0.5698],[380,0.5698]]
       }
       CTRL_L_brow_lateral {
        type=MovieSceneScriptingFloatChannel
        default=3.4028234663852886e+38
        thinned_from=40
        keys [[-10,0.5698],[107,0.5698],[111,0.6095],[117,0.7009],[123,0.6739],[137,0.649],[150,0.649],[156,0.6501],[159,0.6846],[162,0.9353],[165,0.6707],[168,0.5698],[178,0.5106],[183,0.3339],[193,0.3015],[210,0.4827],[231,0.6692],[243,0.7314],[322,0.7314],[329,0.6102],[340,0.5698],[380,0.5698]]
       }
       CTRL_R_brow_down {
        type=MovieSceneScriptingFloatChannel
        default=3.4028234663852886e+38
        thinned_from=42
        keys [[-10,0.471],[0,0.1119],[7,0.0275],[19,0],[39,0],[50,0.1753],[59,0],[107,0],[111,0.25],[117,0.6652],[123,0.6018],[137,0.5802],[150,0.5802],[156,0.6648],[159,0.8368],[162,0.9574],[165,0.3515],[168,0],[178,0.6821],[183,0.604],[193,0.5903],[210,0.688],[219,1],[226,0.4461],[229,0.3995],[231,0.791],[243,0.8334],[253,0.8223],[259,0.9424],[264,0.6599],[274,0.5657],[316,0.5657],[322,0.751],[329,0.1877],[340,0],[355,0.1444],[363,0.471],[373,0.1119],[380,0.0275]]
       }
       CTRL_L_brow_down {
        type=MovieSceneScriptingFloatChannel
        default=3.4028234663852886e+38
        thinned_from=42
        keys [[-10,0.471],[0,0.1119],[7,0.0275],[19,0],[35,0],[39,0.1887],[50,0.4269],[59,0.2516],[107,0.2516],[111,0.4387],[117,1],[123,0.7794],[137,0.5802],[150,0.5802],[156,0.6648],[159,0.8368],[162,0.9574],[165,0.3515],[168,0],[178,0.6821],[183,0.604],[193,0.5903],[210,0.688],[219,1],[226,0.4461],[229,0.3995],[231,0.791],[243,0.8334],[253,0.8223],[259,0.9424],[264,0.6599],[274,0.5657],[316,0.5657],[322,0.751],[329,0.1877],[340,0],[355,0.1444],[363,0.471],[373,0.1119],[380,0.0275]]
       }
       CTRL_L_brow_raiseOut {
        type=MovieSceneScriptingFloatChannel
        default=3.4028234663852886e+38
        thinned_from=40
        keys [[-10,0],[107,0],[111,0.0379],[117,0.1517],[123,0.1787],[137,0.188],[150,0.188],[156,0.1791],[159,0.1645],[165,0.0411],[168,0],[178,0.1607],[193,0.1607],[210,0.2411],[228,0.3133],[231,0.3239],[243,0.3515],[322,0.3515],[329,0.0879],[340,0],[380,0]]
       }
       CTRL_L_brow_raiseIn {
        type=MovieSceneScriptingFloatChannel
        default=3.4028234663852886e+38
        thinned_from=39
        keys [[-10,0.3971],[31,0.3971],[35,0.4646],[39,0.4871],[41,0.4758],[50,0.4168],[59,0.3971],[107,0.3971],[111,0.4368],[117,0.5556],[123,0.5839],[137,0.5935],[150,0.5935],[156,0.5657],[159,0.5193],[165,0.1298],[168,0],[178,0.1607],[193,0.1607],[210,0.2403],[231,0.3224],[243,0.3497],[253,0.3497],[259,0.2356],[264,0.3355],[274,0.3497],[322,0.3497],[329,0.3853],[340,0.3971],[380,0.3971]]
       }
       CTRL_R_brow_raiseIn {
        type=MovieSceneScriptingFloatChannel
        default=3.4028234663852886e+38
        thinned_from=38
        keys [[-10,0.263],[31,0.263],[35,0.4161],[39,0.4671],[50,0.4585],[59,0.4556],[107,0.4556],[117,0.4583],[150,0.4593],[156,0.447],[159,0.4265],[165,0.2539],[168,0.1963],[193,0.1963],[210,0.2699],[231,0.3457],[243,0.371],[253,0.371],[259,0.2569],[264,0.3567],[274,0.371],[322,0.371],[329,0.29],[340,0.263],[380,0.263]]
       }
       CTRL_R_brow_raiseOut {
        type=MovieSceneScriptingFloatChannel
        default=3.4028234663852886e+38
        thinned_from=39
        keys [[-10,0],[31,0],[35,0.0399],[39,0.1214],[50,0.1749],[59,0.1927],[107,0.1927],[111,0.1578],[117,0.0532],[123,0.0135],[137,0],[150,0],[156,0.0106],[159,0.0282],[165,0.1764],[168,0.2258],[193,0.2258],[210,0.1306],[228,0.0453],[231,0.0327],[243,0],[380,0]]
       }
       CTRL_lookAtSwitch {type=MovieSceneScriptingFloatChannel, default=3.4028234663852886e+38, thinned_from=32, keys [[-10,0],[407,0]]}
       CTRL_C_teethD.X {
        type=MovieSceneScriptingFloatChannel
        default=0
        thinned_from=38
        keys [[-10,0],[113,0],[123,-0.7372],[137,-0.9829],[159,-0.9829],[165,-0.0728],[168,0],[174,-0.0572],[179,-0.0583],[213,-0.0583],[219,-0.7518],[231,-0.9829],[244,-0.9956],[254,-0.9956],[264,-0.9858],[290,-0.9829],[321,-0.9829],[341,-0.2457],[363,0],[380,0]]
       }
       CTRL_C_teethD.Y {
        type=MovieSceneScriptingFloatChannel
        default=0
        thinned_from=38
        keys [[-10,0],[113,0],[123,0.75],[137,1],[159,1],[165,0.1023],[168,0.0305],[174,0.1818],[179,0.1835],[213,0.1835],[219,0.7959],[231,1],[321,1],[341,0.25],[363,0],[380,0]]
       }
       CTRL_C_teethU.X {
        type=MovieSceneScriptingFloatChannel
        default=0
        thinned_from=38
        keys [[-10,0],[113,0],[123,-0.7372],[137,-0.9829],[159,-0.9829],[165,-0.0728],[168,0],[213,0],[219,-0.7372],[231,-0.9829],[244,-0.9964],[254,-0.9964],[264,-0.9859],[290,-0.9829],[321,-0.9829],[341,-0.2457],[363,0],[380,0]]
       }
       CTRL_C_teethU.Y {
        type=MovieSceneScriptingFloatChannel
        default=0
        thinned_from=38
        keys [[-10,0],[113,0],[123,0.75],[137,1],[159,1],[165,-0.0479],[168,-0.1317],[174,-0.2147],[179,-0.2069],[213,-0.2069],[219,0.6983],[231,1],[321,1],[341,0.25],[363,0],[380,0]]
       }
       CTRL_C_teeth_fwdBackU {
        type=MovieSceneScriptingFloatChannel
        default=3.4028234663852886e+38
        thinned_from=39
        keys [[-10,0],[113,0],[123,0.75],[137,1],[159,1],[165,0.0741],[168,0],[213,0],[219,0.75],[231,1],[321,1],[341,0.25],[363,0],[380,0]]
       }
       CTRL_L_mouth_corner.X {
        type=MovieSceneScriptingFloatChannel
        default=0.2435
        thinned_from=41
        keys [[-10,0.2435],[37,0.2435],[43,0.2079],[52,0.038],[61,-0.1536],[107,-0.1536],[113,-0.0831],[123,0.4467],[137,0.6369],[198,0.6369],[202,0.6816],[205,0.7798],[207,0.8544],[208,0.9818],[213,1],[231,1],[244,0.9854],[254,0.9854],[259,0.9134],[264,0.4094],[290,0.2435],[380,0.2435]]
       }
       CTRL_L_mouth_corner.Y {
        type=MovieSceneScriptingFloatChannel
        default=0.3239
        thinned_from=41
        keys [[-10,0.3239],[37,0.3239],[43,0.2921],[52,0.0349],[61,-0.1209],[84,-0.0776],[107,-0.0776],[113,-0.0146],[123,0.4586],[137,0.6285],[150,0.6285],[159,0.6607],[165,0.9756],[168,1],[198,1],[202,0.9744],[205,0.9181],[207,0.8753],[208,0.8023],[213,0.7918],[254,0.7909],[259,0.7456],[264,0.4283],[290,0.3239],[380,0.3239]]
       }
       CTRL_R_mouth_corner.X {
        type=MovieSceneScriptingFloatChannel
        default=0
        thinned_from=38
        keys [[-10,0],[37,0],[43,0.0116],[52,0.0927],[61,0.1337],[107,0.1337],[113,0.219],[123,0.4108],[137,0.4747],[150,0.4747],[198,0.4739],[202,0.3405],[208,-0.5935],[213,-0.6179],[231,-0.6179],[244,-0.607],[254,-0.607],[264,-0.1358],[290,0],[380,0]]
       }
       CTRL_R_mouth_corner.Y {
        type=MovieSceneScriptingFloatChannel
        default=0.831
        thinned_from=38
        keys [[-10,0.831],[37,0.831],[43,0.754],[52,0.3432],[61,0.0685],[84,0.0408],[107,0.0408],[113,0.1684],[123,0.4556],[137,0.5513],[150,0.5513],[198,0.5522],[202,0.5631],[208,0.6394],[213,0.6846],[254,0.6839],[264,0.7981],[290,0.831],[380,0.831]]
       }
       CTRL_C_teeth_fwdBackD {
        type=MovieSceneScriptingFloatChannel
        default=3.4028234663852886e+38
        thinned_from=39
        keys [[-10,0],[113,0],[123,0.75],[137,1],[159,1],[165,0.1353],[168,0.0661],[174,-0.1877],[177,-0.2047],[179,-0.2195],[182,-0.2229],[213,-0.2229],[219,0.6943],[231,1],[321,1],[341,0.25],[363,0],[380,0]]
       }
       CTRL_R_mouth_cornerSharpnessD {
        type=MovieSceneScriptingFloatChannel
        default=3.4028234663852886e+38
        thinned_from=38
        keys [[-10,0],[159,0],[165,-0.4913],[168,-0.5306],[198,-0.5306],[202,-0.4653],[205,-0.3218],[207,-0.2128],[208,-0.0266],[213,0],[380,0]]
       }
       CTRL_R_mouth_cornerSharpnessU {type=MovieSceneScriptingFloatChannel, default=3.4028234663852886e+38, const [0,38]}
       CTRL_L_eye_eyelidU.X {
        type=MovieSceneScriptingFloatChannel
        default=0
        thinned_from=57
        keys [[-10,0],[114,0],[117,-0.2121],[121,-0.5388],[124,-0.7064],[137,-0.7226],[159,-0.7226],[165,-0.6154],[168,-0.6154],[172,-0.5848],[174,-0.2988],[193,-0.2988],[210,-0.4774],[231,-0.7226],[253,-0.7226],[256,-0.6116],[259,0],[264,-0.7226],[316,-0.7226],[320,0],[380,0]]
       }
       CTRL_L_eye_eyelidU.Y {
        type=MovieSceneScriptingFloatChannel
        default=0.0189
        thinned_from=58
        keys [[-10,-0.3375],[-5,-0.0679],[0,0.0189],[7,0.0627],[19,0.0769],[34,0.0769],[35,-0.0103],[37,-0.2339],[39,-0.3375],[50,0.0769],[107,0.0769],[111,-0.3375],[114,-0.3375],[117,-0.2168],[121,-0.0308],[124,-0.1985],[137,-0.2146],[159,-0.2146],[165,0.3467],[168,0.3467],[172,0.3003],[174,-0.1343],[175,-0.3066],[177,-0.1788],[183,-0.1398],[193,-0.1343],[210,-0.0792],[231,-0.0087],[253,-0.0087],[256,-0.1117],[259,-0.6796],[264,-0.0087],[316,-0.0087],[320,-0.1827],[322,-0.1827],[327,-0.0699],[340,0.0769],[358,0.0769],[359,-0.0103],[361,-0.2339],[363,-0.3375],[368,-0.0679],[373,0.0189],[380,0.0627]]
       }
       CTRL_L_mouth_cornerSharpnessD {
        type=MovieSceneScriptingFloatChannel
        default=3.4028234663852886e+38
        thinned_from=38
        keys [[-10,0],[159,0],[165,-0.6562],[168,-0.7086],[174,-0.5306],[198,-0.5306],[202,-0.4554],[205,-0.29],[207,-0.1645],[208,0.05],[213,0.0806],[231,0.0806],[244,0.0797],[254,0.0797],[264,0.0178],[290,0],[321,-0.0025],[341,-0.0006],[363,0],[380,0]]
       }
       CTRL_L_eye_eyelidD.X {type=MovieSceneScriptingFloatChannel, default=0, const [0,57]}
       CTRL_L_eye_eyelidD.Y {
        type=MovieSceneScriptingFloatChannel
        default=-0.0495
        thinned_from=58
        keys [[-10,-0.4145],[-5,-0.1384],[0,-0.0495],[7,-0.0122],[19,0],[34,0],[35,-0.0873],[37,-0.3108],[39,-0.4145],[50,0],[107,0],[111,-0.4145],[114,-0.4145],[117,-0.2513],[121,0],[159,0],[165,0.2058],[168,0.2058],[172,0.1859],[174,0],[175,-0.0737],[177,-0.0191],[183,-0.0024],[193,0],[210,0.0877],[231,0.2059],[253,0.2059],[256,0.07],[259,-0.6796],[264,0.2059],[316,0.2059],[320,0.0319],[322,0.0319],[327,-0.1276],[340,0],[358,0],[359,-0.0873],[361,-0.3108],[363,-0.4145],[368,-0.1384],[373,-0.0495],[380,-0.0122]]
       }
       CTRL_L_mouth_cornerSharpnessU {
        type=MovieSceneScriptingFloatChannel
        default=3.4028234663852886e+38
        thinned_from=38
        keys [[-10,0],[198,0],[202,0.0099],[205,0.0317],[207,0.0483],[208,0.0766],[213,0.0806],[231,0.0806],[244,0.0797],[254,0.0797],[264,0.0178],[290,0],[321,-0.0025],[341,-0.0006],[363,0],[380,0]]
       }
       CTRL_R_mouth_lipsTowardsTeethD {type=MovieSceneScriptingFloatChannel, default=3.4028234663852886e+38, const [0,38]}
       CTRL_L_mouth_lipsTowardsTeethD {type=MovieSceneScriptingFloatChannel, default=3.4028234663852886e+38, const [0,38]}
       CTRL_R_eye_eyelidU.X {
        type=MovieSceneScriptingFloatChannel
        default=0
        thinned_from=57
        keys [[-10,0],[114,0],[117,-0.1176],[121,-0.2988],[124,-0.4665],[137,-0.4826],[159,-0.4826],[165,-0.1207],[168,-0.1207],[172,-0.1379],[174,-0.2988],[175,-0.3627],[177,-0.3154],[183,-0.3009],[193,-0.2988],[210,-0.3323],[231,-0.3782],[253,-0.3782],[256,-0.3201],[259,0],[264,-0.3782],[316,-0.3782],[320,0],[380,0]]
       }
       CTRL_R_eye_eyelidU.Y {
        type=MovieSceneScriptingFloatChannel
        default=0.0189
        thinned_from=58
        keys [[-10,-0.3375],[-5,-0.0679],[0,0.0189],[7,0.0627],[19,0.0769],[34,0.0769],[35,-0.0103],[37,-0.2339],[39,-0.3375],[50,0.0769],[107,0.0769],[111,-0.3375],[114,-0.3375],[117,-0.2575],[121,-0.1343],[124,-0.3019],[137,-0.318],[159,-0.318],[165,0.1262],[168,0.1262],[172,0.1011],[174,-0.1343],[175,-0.2276],[177,-0.1584],[183,-0.1373],[193,-0.1343],[210,-0.1238],[231,-0.1122],[253,-0.1122],[256,-0.1993],[259,-0.6796],[264,-0.1122],[316,-0.1122],[320,-0.2861],[322,-0.2861],[327,-0.0699],[340,0.0769],[358,0.0769],[359,-0.0103],[361,-0.2339],[363,-0.3375],[368,-0.0679],[373,0.0189],[380,0.0627]]
       }
       CTRL_L_mouth_lipsTowardsTeethU {type=MovieSceneScriptingFloatChannel, default=3.4028234663852886e+38, const [0,38]}
       CTRL_R_eye_eyelidD.X {type=MovieSceneScriptingFloatChannel, default=0, const [0,57]}
       CTRL_R_eye_eyelidD.Y {
        type=MovieSceneScriptingFloatChannel
        default=-0.0607
        thinned_from=58
        keys [[-10,-0.4145],[-5,-0.1468],[0,-0.0607],[7,-0.0149],[19,0],[34,0],[35,-0.0873],[37,-0.3108],[39,-0.4145],[50,0],[107,0],[111,-0.4145],[114,-0.4145],[117,-0.2513],[121,0],[159,0],[165,0.2058],[168,0.2058],[172,0.1859],[174,0],[175,-0.0737],[177,-0.0191],[183,-0.0024],[193,0],[210,0.0877],[231,0.2059],[253,0.2059],[256,0.07],[259,-0.6796],[264,0.2059],[316,0.2059],[320,0.0319],[322,0.0319],[327,-0.1276],[340,0],[358,0],[359,-0.0873],[361,-0.3108],[363,-0.4145],[368,-0.1468],[373,-0.0607],[380,-0.0149]]
       }
       CTRL_R_mouth_lipsTowardsTeethU {type=MovieSceneScriptingFloatChannel, default=3.4028234663852886e+38, const [0,38]}
       CTRL_faceGUIfollowHead {type=MovieSceneScriptingFloatChannel, default=3.4028234663852886e+38, const [0,27]}
       CTRL_eyesAimFollowHead {type=MovieSceneScriptingFloatChannel, default=3.4028234663852886e+38, const [0,27]}
       CTRL_C_nose.X {type=MovieSceneScriptingFloatChannel, default=0, const [0,27]}
       CTRL_C_nose.Y {type=MovieSceneScriptingFloatChannel, default=0, const [0,27]}
       CTRL_C_teeth_rotateU {
        type=MovieSceneScriptingFloatChannel
        default=3.4028234663852886e+38
        thinned_from=38
        keys [[-10,0],[113,0],[123,0.75],[137,1],[159,1],[165,-0.1167],[168,-0.206],[174,-0.4798],[213,-0.4798],[219,0.6301],[231,1],[321,1],[341,0.25],[363,0],[380,0]]
       }
       CTRL_C_teeth_rotateD {
        type=MovieSceneScriptingFloatChannel
        default=3.4028234663852886e+38
        thinned_from=38
        keys [[-10,0],[113,0],[123,0.75],[137,1],[159,1],[165,-0.1167],[168,-0.206],[174,0.0519],[179,0.0597],[213,0.0597],[219,0.7649],[231,1],[321,1],[341,0.25],[363,0],[380,0]]
       }
       CTRL_L_brow_inner {
        type=MovieSceneScriptingFloatChannel
        default=3.4028234663852886e+38
        thinned_from=40
        keys [[-10,0],[31,0],[35,0.0345],[39,0.0951],[41,0.0448],[50,-0.1305],[59,-0.242],[107,-0.242],[111,-0.2628],[117,-0.3253],[123,-0.3284],[137,-0.3295],[150,-0.3295],[156,-0.4272],[159,-0.59],[165,0.16],[168,0],[178,0.0225],[183,0.0895],[193,0.1018],[210,-0.0799],[219,-0.4716],[231,-1],[243,-0.8377],[253,-0.7961],[259,-0.6795],[264,-0.417],[274,-0.3295],[322,-0.3295],[329,-0.0824],[340,0],[380,0]]
       }
       CTRL_R_brow_inner {
        type=MovieSceneScriptingFloatChannel
        default=3.4028234663852886e+38
        thinned_from=40
        keys [[-10,-0.4266],[35,-0.4266],[39,-0.1408],[41,-0.1255],[50,-0.0449],[59,-0.0181],[107,-0.0181],[111,-0.0348],[117,-0.1521],[123,-0.486],[137,-0.5999],[150,-0.5999],[156,-0.6849],[159,-0.8266],[165,0.1009],[168,0],[178,0.0225],[183,0.0895],[193,0.1018],[210,-0.1939],[219,-0.6767],[231,-1],[243,-0.8459],[253,-0.8064],[259,-0.7547],[264,-0.6386],[274,-0.5999],[322,-0.5999],[329,-0.4699],[340,-0.4266],[380,-0.4266]]
       }
       CTRL_L_brow_outer {
        type=MovieSceneScriptingFloatChannel
        default=3.4028234663852886e+38
        thinned_from=40
        keys [[-10,0],[31,0],[35,0.0345],[39,0.0951],[41,0.0448],[50,-0.1305],[59,-0.242],[107,-0.242],[111,-0.3028],[117,-0.3759],[123,-0.3905],[137,-0.3955],[150,-0.3955],[156,-0.3934],[159,-0.3899],[165,0.2139],[168,0.4152],[178,0.0581],[183,0.15],[193,0.1663],[210,-0.1165],[219,-0.5891],[231,-0.1335],[243,-0.5049],[322,-0.5049],[329,-0.1262],[340,0],[380,0]]
       }
       CTRL_R_brow_outer {
        type=MovieSceneScriptingFloatChannel
        default=3.4028234663852886e+38
        thinned_from=40
        keys [[-10,-0.4266],[35,-0.4266],[39,-0.1408],[41,-0.1255],[50,-0.0449],[59,-0.0181],[107,-0.0181],[111,-0.0616],[117,-0.1922],[123,-0.627],[137,-0.7753],[150,-0.7753],[156,-0.728],[159,-0.6491],[165,0.0134],[168,0.2343],[178,-0.4777],[183,-0.2943],[193,-0.262],[210,-0.4783],[219,-0.8976],[231,-0.8059],[243,-0.7753],[322,-0.7753],[329,-0.5138],[340,-0.4266],[380,-0.4266]]
       }
       CTRL_C_mouth.X {
        type=MovieSceneScriptingFloatChannel
        default=0
        thinned_from=40
        keys [[-10,0],[37,0],[43,0.0128],[52,0.1082],[61,0.1859],[113,0.1859],[123,0.0465],[137,0],[150,0],[159,0.0289],[165,0.0021],[168,0],[174,0.2312],[198,0.2312],[202,0.2367],[205,0.2488],[207,0.258],[208,0.2737],[213,0.2759],[219,0.2667],[231,0.2019],[244,0.1913],[254,0.1728],[264,0.0428],[290,0],[380,0]]
       }
       CTRL_C_mouth.Y {
        type=MovieSceneScriptingFloatChannel
        default=0.0598
        thinned_from=40
        keys [[-10,0.0598],[37,0.0598],[43,0.035],[52,0.0525],[61,0.0598],[113,0.0598],[123,0.2026],[137,0.2502],[150,0.2502],[159,0.2122],[165,-0.1188],[168,-0.1443],[174,-0.0453],[179,-0.0541],[198,-0.0541],[202,-0.0829],[205,-0.1463],[207,-0.1945],[208,-0.2767],[213,-0.2884],[219,-0.273],[231,-0.1653],[244,-0.1477],[254,-0.1583],[259,-0.1884],[264,-0.0055],[290,0.0598],[380,0.0598]]
       }
       CTRL_L_mouth_upperLipRaise {
        type=MovieSceneScriptingFloatChannel
        default=3.4028234663852886e+38
        thinned_from=40
        keys [[-10,0],[159,0],[165,0.1479],[168,0.1598],[174,0],[380,0]]
       }
       CTRL_L_mouth_cornerPull {
        type=MovieSceneScriptingFloatChannel
        default=3.4028234663852886e+38
        thinned_from=39
        keys [[-10,0],[113,0],[123,0.0582],[137,0.0776],[150,0.0776],[159,0.0679],[165,0.0048],[168,0],[174,0.1572],[179,0.164],[198,0.164],[202,0.1781],[205,0.2092],[207,0.2328],[208,0.2732],[213,0.279],[231,0.279],[244,0.2774],[254,0.2774],[259,0.2509],[264,0.1714],[290,0.1408],[321,0.1375],[341,0.0344],[363,0],[380,0]]
       }
       CTRL_L_mouth_dimple {
        type=MovieSceneScriptingFloatChannel
        default=3.4028234663852886e+38
        thinned_from=41
        keys [[-10,0.0398],[37,0.0397],[43,0.0468],[52,0.0959],[61,0.1208],[113,0.1208],[123,0.06],[137,0.0397],[150,0.0397],[159,0.0701],[165,0.7425],[168,0.7955],[174,0.6183],[179,0.669],[198,0.669],[202,0.5994],[205,0.4462],[207,0.3299],[208,0.1459],[213,0.1176],[231,0.1176],[254,0.1173],[259,0.1023],[264,0.0571],[290,0.0397],[380,0.0397]]
       }
       CTRL_L_mouth_cornerDepress {
        type=MovieSceneScriptingFloatChannel
        default=3.4028234663852886e+38
        thinned_from=39
        keys [[-10,0],[168,0],[174,0.0887],[198,0.0887],[202,0.0778],[205,0.0538],[207,0.0356],[208,0.0044],[213,0],[380,0]]
       }
       CTRL_L_mouth_stretch {type=MovieSceneScriptingFloatChannel, default=3.4028234663852886e+38, const [0,40]}
       CTRL_L_mouth_lowerLipDepress {
        type=MovieSceneScriptingFloatChannel
        default=3.4028234663852886e+38
        thinned_from=40
        keys [[-10,0],[159,0],[165,0.2803],[168,0.3028],[198,0.3028],[202,0.2655],[205,0.1836],[207,0.1214],[208,0.0152],[213,0],[380,0]]
       }
       CTRL_R_mouth_lowerLipDepress {
        type=MovieSceneScriptingFloatChannel
        default=3.4028234663852886e+38
        thinned_from=39
        keys [[-10,0],[159,0],[165,0.2803],[168,0.3028],[198,0.3028],[202,0.2655],[205,0.1836],[207,0.0686],[208,0.0152],[213,0],[380,0]]
       }
       CTRL_R_mouth_stretch {type=MovieSceneScriptingFloatChannel, default=3.4028234663852886e+38, const [0,39]}
       CTRL_L_mouth_towardsD {type=MovieSceneScriptingFloatChannel, default=3.4028234663852886e+38, const [0,39]}
       CTRL_R_mouth_funnelU {type=MovieSceneScriptingFloatChannel, default=3.4028234663852886e+38, const [0,39]}
       CTRL_L_mouth_towardsU {type=MovieSceneScriptingFloatChannel, default=3.4028234663852886e+38, const [0,39]}
       CTRL_R_mouth_lipBiteD {type=MovieSceneScriptingFloatChannel, default=3.4028234663852886e+38, const [0,29]}
       CTRL_R_mouth_funnelD {type=MovieSceneScriptingFloatChannel, default=3.4028234663852886e+38, const [0,39]}
       CTRL_L_mouth_pressU {type=MovieSceneScriptingFloatChannel, default=3.4028234663852886e+38, const [0,39]}
       CTRL_L_mouth_funnelD {type=MovieSceneScriptingFloatChannel, default=3.4028234663852886e+38, const [0,39]}
       CTRL_L_mouth_suckBlow {
        type=MovieSceneScriptingFloatChannel
        default=3.4028234663852886e+38
        thinned_from=39
        keys [[-10,0],[150,0],[159,0.0445],[165,0.2815],[168,0.2993],[174,-0.1473],[198,-0.1473],[202,-0.1292],[205,-0.0894],[207,-0.0591],[208,-0.0074],[213,0],[380,0]]
       }
       CTRL_L_mouth_pressD {type=MovieSceneScriptingFloatChannel, default=3.4028234663852886e+38, const [0,39]}
       CTRL_L_mouth_funnelU {type=MovieSceneScriptingFloatChannel, default=3.4028234663852886e+38, const [0,39]}
       CTRL_R_mouth_towardsU {type=MovieSceneScriptingFloatChannel, default=3.4028234663852886e+38, const [0,38]}
       CTRL_R_mouth_purseD {type=MovieSceneScriptingFloatChannel, default=3.4028234663852886e+38, const [0,39]}
       CTRL_L_mouth_lipBiteU {type=MovieSceneScriptingFloatChannel, default=3.4028234663852886e+38, const [0,29]}
       CTRL_R_mouth_pressD {type=MovieSceneScriptingFloatChannel, default=3.4028234663852886e+38, const [0,39]}
       CTRL_R_mouth_pressU {type=MovieSceneScriptingFloatChannel, default=3.4028234663852886e+38, const [0,39]}
       CTRL_L_mouth_purseD {type=MovieSceneScriptingFloatChannel, default=3.4028234663852886e+38, const [0,39]}
       CTRL_R_mouth_lipBiteU {type=MovieSceneScriptingFloatChannel, default=3.4028234663852886e+38, const [0,29]}
       CTRL_R_mouth_suckBlow {
        type=MovieSceneScriptingFloatChannel
        default=3.4028234663852886e+38
        thinned_from=39
        keys [[-10,0],[150,0],[159,0.018],[165,0.0013],[168,0],[380,0]]
       }
       CTRL_L_mouth_purseU {type=MovieSceneScriptingFloatChannel, default=3.4028234663852886e+38, const [0,39]}
       CTRL_R_mouth_towardsD {type=MovieSceneScriptingFloatChannel, default=3.4028234663852886e+38, const [0,38]}
       CTRL_L_mouth_lipBiteD {type=MovieSceneScriptingFloatChannel, default=3.4028234663852886e+38, const [0,29]}
       CTRL_R_mouth_purseU {type=MovieSceneScriptingFloatChannel, default=3.4028234663852886e+38, const [0,39]}
       CTRL_R_mouth_cornerDepress {type=MovieSceneScriptingFloatChannel, default=3.4028234663852886e+38, const [0,39]}
       CTRL_R_mouth_upperLipRaise {
        type=MovieSceneScriptingFloatChannel
        default=3.4028234663852886e+38
        thinned_from=39
        keys [[-10,0],[159,0],[165,0.172],[168,0.1857],[174,0],[380,0]]
       }
       CTRL_R_mouth_dimple {
        type=MovieSceneScriptingFloatChannel
        default=3.4028234663852886e+38
        thinned_from=39
        keys [[-10,0.3986],[37,0.3986],[43,0.3641],[52,0.1224],[61,0],[113,0],[123,0.2912],[137,0.3882],[150,0.3882],[159,0.4013],[165,0.8693],[168,0.9064],[174,0.776],[179,0.8832],[198,0.8832],[202,0.8223],[205,0.6884],[208,0.4469],[213,0.3882],[254,0.3882],[264,0.3963],[290,0.3986],[380,0.3986]]
       }
       CTRL_R_mouth_cornerPull {
        type=MovieSceneScriptingFloatChannel
        default=3.4028234663852886e+38
        thinned_from=39
        keys [[-10,0],[150,0],[159,0.006],[165,0.0449],[168,0.0479],[174,0.2001],[198,0.2001],[202,0.1755],[205,0.1213],[207,0.0802],[208,0.0294],[213,0],[380,0]]
       }
      ]
     }]
   }
  ]
 }]
display_rate {denominator=1, numerator=30}
tick_resolution {denominator=1, numerator=24000}
tracks []
